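# Opaque lit, Nanite-used material for liquid effects on the industrial processing machine B prop.
# World position scaled by WorldTiling and panned over looping time drives tiling 3D noise, pushed along vertex normals by WPO_Intensity for surface motion.
# A shading path switch uses texture-based gradient noise by default and ALU value noise on forward and mobile.
# The machine's diffuse, normal and mask textures are also sampled.
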Material /Hermes/Props/Industrial/Industrial_ProcessingMachine_B/FX/M_Industrial_ProcessingMachine_FX_Liquid
graph {
 properties {assetImportData="<self>:InterchangeAssetImportData_0", bUsedWithNanite=true, usedShadingModels=Default Lit}
 nodes [
  Constant3Vector() {constant=<rgba 0.7569 0.7569 0.7569 1>}
  n1 = Add(
   A=Multiply(
    A=Time() {bIgnorePause=true, bOverride_Period=true, period=100}
    B=VectorParameter() {parameterName=PanSpeed, defaultValue=<rgba 250 1500 -500 1>}.RGB
   ) {constB=250}
   B=Divide(A=WorldPosition().XYZ, B=VectorParameter() {parameterName=WorldTiling, defaultValue=<rgba 0.02 0.02 0.02 1>}.RGB)
  )
  n2 = Noise!(World Position=@n1, FilterWidth) {bTiling=true, bTurbulence=false, levelScale=1, levels=1, noiseFunction=NOISEFUNCTION_GradientTex3D, repeatSize=500, scale=0.002}
  n3 = Noise(World Position=@n1) {bTiling=true, bTurbulence=false, levelScale=1, levels=2, noiseFunction=NOISEFUNCTION_ValueALU, repeatSize=500, scale=0.002}
  Multiply(
   A=VertexNormalWS()
   B=Multiply(A=ShadingPathSwitch(Default=@n2, Forward=@n3, Mobile=@n3), B=ScalarParameter() {defaultValue=1, parameterName=WPO_Intensity})
  )
  ScalarParameter() {defaultValue=0.1}
  TextureSample() {texture=/Hermes/Props/Industrial/Industrial_ProcessingMachine_B/Textures/T_Industrial_ProcessingMachine_B_B_D}
  TextureSample() {
   samplerType=SAMPLERTYPE_Normal
   texture=/Hermes/Props/Industrial/Industrial_ProcessingMachine_B/Textures/T_Industrial_ProcessingMachine_B_B_N
  }
  TextureSample() {samplerType=SAMPLERTYPE_Masks, texture=/Hermes/Props/Industrial/Industrial_ProcessingMachine_B/Textures/T_Industrial_ProcessingMachine_B_B_S}
 ]
}
